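# Niagara system for the Magma vehicle boost: fire, core burst, swirls, drops and ribbon trails streaming from car sockets.
# Emitters partition particles across skeletal mesh sockets by port count, shift them by a car offset, and pick colors through painted-variant overrides.
# Spawn groups split fire and swirl into burst and continuous behaviors.
# Drops generate collision events and die on impact; a blob emitter has spawn rate disabled, suggesting event-driven spawning.
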NiagaraSystem /VehicleCosmetics/Boosts/Boost_Magma/FX/NS_Boost_Magma
dependencies {
 usedDataInterfaces [
  &NiagaraDataInterfaceVectorCurve
  &NiagaraDataInterfaceCurve
  &NiagaraDataInterfaceSpriteRendererInfo
  &NiagaraDataInterfaceSkeletalMesh
 ]
 usedDynamicInputs []
 usedModules [
  &self:SystemSpawnScript.SystemScriptSource.SystemScriptGraph.NiagaraNodeAssignment_0.SetVariables_#1
  &SystemState
 ]
}
emitters {
 Collided_blobs {
  emitterUpdateScript [
   &EmitterState(Inactive Response=::NewEnumerator1, Life Cycle Mode=::NewEnumerator1, UseLoopDelay=false)
   &VC_FilteredSocketCount()
   &SpawnRate(SpawnRate=50, Use Spawn Probability=false) {enabled=false}
  ]
  particleSpawnScript [
   &VC_PaintedVariantsOverride(
    Black=<rgba 0.1 0.1 0.1 1>
    BurntSienna=<rgba 0.55 0.11 0 1>
    Cobalt=<rgba 0 0.4167 1 1>
    Crimson=<rgba 1 0 0 1>
    DefaultColor=<rgba 2 0.0667 0 1>
    ForestGreen=<rgba 0 0.8 0 1>
    Grey=<rgba 0.5 0.5 0.5 1>
    Lime=<rgba 0.925 1.5 0 1>
    Orange=<rgba 1.3 0.1733 0 1>
    Pink=<rgba 1.2 0.36 1.06 1>
    Purple=<rgba 0.4167 0 1 1>
    Saffron=<rgba 1.2133 1.3 0 1>
    SkyBlue=<rgba 0 0.65 1.3 1>
   )
   &InitializeParticle(
    Color=$Particles.Color
    Lifetime Max=0.8
    Lifetime Min=0.6
    Lifetime Mode=::NewEnumerator1
    Mesh Scale Max=<1 1 1>
    Mesh Scale Min=<0.5 0.5 0.75>
    Mesh Scale Mode=::NewEnumerator2
    UsePositionOffset=false
    Write Material Random=true
   )
   &PartitionParticles(Number Of Partitions=$Emitter.PortCount)
   &SkeletalMeshLocation(
    Bone / Socket Index=$Particles.PartitionParticles.Partition
    Bone / Socket Sampling Mode=::NewEnumerator5
    MaskSpawnGroup=false
    Mesh Sampling Type=::NewEnumerator6
    Sampled Position Offset=&AddVectorToPosition(Position=<0 0 0>, Vector=$User.DTCarOffset)
    Skeletal Mesh=$System.SkeletalMesh
   )
   &AddVelocity() {enabled=false}
   &InitialMeshOrientation(
    Mesh Orientation Mode=::NewEnumerator7
    Rotation=&RandomRangeVector(Minimum=<0 0 0>, Maximum=<0 0 1>)
    Use Rotation Vector=true
   )
  ]
  particleUpdateScript [
   &ParticleState()
   &ScaleMeshSize(Scale Factor=&VectorFromCurve(VectorCurve=DataInterface() {propertyValues {
       XCurve {keys [
         <at 0 0.1>
         <at 0.05 0.35>
         {arriveTangent=0.1579, interpMode=RCIM_Cubic, leaveTangent=0.1579, time=0.1, value=1}
         {arriveTangent=-0.1029, interpMode=RCIM_Cubic, leaveTangent=-0.1029, tangentMode=RCTM_User, time=1, value=0.5}
        ]}
       YCurve {keys [<at 0 0.1>, <at 0.035 0.4>, {arriveTangent=0.1036, interpMode=RCIM_Cubic, leaveTangent=0.1036, time=0.08, value=1}, <at 1 0.5 RCIM_Cubic>]}
       ZCurve {keys [
         <at 0 4 RCIM_Cubic>
         {arriveTangent=-3.6817, interpMode=RCIM_Cubic, leaveTangent=-3.6817, tangentMode=RCTM_User, time=0.12, value=1}
         {arriveTangent=-0.0895, interpMode=RCIM_Cubic, leaveTangent=-0.0895, tangentMode=RCTM_User, time=1, value=0}
        ]}
      }}))
   &SolveForcesAndVelocity(Clamp Velocity=false, Limit Acceleration=false)
   &ScaleColor(
    Scale RGB=&VectorFromFloat(Value=&Multiply_Float(
      A=6
      B=&FloatFromCurve(FloatCurve=DataInterface() {propertyValues {Curve {keys [
           {arriveTangent=-0.1, interpMode=RCIM_Cubic, leaveTangent=-0.1, leaveTangentWeight=0.3333, tangentMode=RCTM_User, time=0, value=1}
           {arriveTangentWeight=0.3333, interpMode=RCIM_Cubic, tangentMode=RCTM_User, time=1, value=0.04}
          ]}}})
     ))
    ScaleA=false
    ScaleRGB=true
   )
  ]
  rendererClasses [&NiagaraMeshRendererProperties, &NiagaraSpriteRendererProperties]
  renderers [
   {rendererClass=&NiagaraMeshRendererProperties, rendererIndex=0}
   {rendererClass=&NiagaraSpriteRendererProperties, rendererIndex=1}
  ]
 }
 NE_CORE_BURST {
  emitterUpdateScript [
   &EmitterState(Inactive Response=::NewEnumerator1, Life Cycle Mode=::NewEnumerator1, Loop Duration=0.1, UseLoopDelay=false)
   &VC_FilteredSocketCount()
   &SpawnRate(SpawnRate=80, Use Spawn Probability=false)
  ]
  particleSpawnScript [
   &VC_PaintedVariantsOverride(
    Black=<rgba 0.01 0.01 0.01 1>
    BurntSienna=<rgba 0.9 0.18 0 1>
    Cobalt=<rgba 0.175 0.3412 3.5 1>
    Crimson=&RandomRangeLinearColor(Minimum=<rgba 0.1 0 0 1>, Maximum=<rgba 1.5 0.045 0.045 1>)
    DefaultColor=&RandomRangeLinearColor(Minimum=<rgba 0.5 0.0083 0 1>, Maximum=<rgba 3 0.45 0 1>)
    ForestGreen=<rgba 0.03 1 0.03 1>
    Grey=<rgba 0.35 0.35 0.35 1>
    Lime=<rgba 0.925 1.5 0 1>
    Orange=<rgba 1.5 0.2 0 1>
    Pink=<rgba 1.8 0.54 1.59 1>
    Purple=&RandomRangeLinearColor(Minimum=<rgba 0.1667 0 0.4 1>, Maximum=<rgba 0.2917 0 0.7 1>)
    Saffron=<rgba 1.4 1.5 0 1>
    SkyBlue=<rgba 0 0.9 1.8 1>
    TitaniumWhite=<rgba 0.8 0.8 0.8 1>
   )
   &InitializeParticle(
    Color=$Particles.Color
    Lifetime Max=0.4
    Lifetime Min=0.3
    Lifetime Mode=::NewEnumerator1
    Position Mode=::NewEnumerator0
    Sprite Rotation Mode=::NewEnumerator2
    Sprite Size Mode=::NewEnumerator1
    Uniform Sprite Size Max=65
    Uniform Sprite Size Min=55
    Write Material Random=true
   )
   &PartitionParticles(Number Of Partitions=$Emitter.PortCount)
   &SkeletalMeshLocation(
    Bone / Socket Index=$Particles.PartitionParticles.Partition
    Bone / Socket Sampling Mode=::NewEnumerator5
    MaskSpawnGroup=false
    Mesh Sampling Type=::NewEnumerator6
    Sampled Position Offset=&AddVectorToPosition(Position=<-85 0 0>, Vector=$User.DTCarOffset)
    Skeletal Mesh=$System.SkeletalMesh
   )
   &AddVelocity(
    Cone Axis=<-1 0 0>
    Inner Cone Angle=5
    Module.UseConeFalloff=false
    Module.UseVelDistribution=false
    Override Seed=false
    Velocity Mode=::NewEnumerator2
    Velocity Speed=&RandomRangeFloat(Minimum=0, Maximum=200)
   )
   &ShapeLocation(MaskSpawnGroup=false, Override Seed=false, Sphere Radius=7.5, UseOwnerScale=true)
  ]
  particleUpdateScript [
   &ParticleState()
   &ScaleSpriteSize(Uniform Curve Sprite Scale=DataInterface() {propertyValues {Curve {
       defaultValue=0
       keys [
        {arriveTangent=4.1785, interpMode=RCIM_Cubic, leaveTangent=4.1785, tangentMode=RCTM_User, time=0, value=0}
        {arriveTangent=1.1179, interpMode=RCIM_Cubic, leaveTangent=1.1179, tangentMode=RCTM_User, time=0.1, value=0.8}
        {arriveTangent=0.0224, interpMode=RCIM_Cubic, leaveTangent=0.0224, tangentMode=RCTM_User, time=1, value=1}
       ]
      }}})
   &SolveForcesAndVelocity(Clamp Velocity=false, Limit Acceleration=false)
   &ScaleColor(
    Scale Alpha=&FloatFromCurve(FloatCurve=DataInterface() {propertyValues {Curve {
        defaultValue=1
        keys [
         {interpMode=RCIM_Cubic, tangentMode=RCTM_User, time=0, value=0}
         {interpMode=RCIM_Cubic, tangentMode=RCTM_User, time=0.5, value=1}
         {tangentMode=RCTM_User, time=1, value=0}
        ]
       }}})
    Scale RGB=&VectorFromFloat(Value=&FloatFromCurve(
      FloatCurve=DataInterface() {propertyValues {Curve {
         defaultValue=1
         keys [
          {arriveTangent=-2.5825, interpMode=RCIM_Cubic, leaveTangent=-2.5825, leaveTangentWeight=0.3333, tangentMode=RCTM_User, time=0, value=1.25}
          {arriveTangentWeight=0.3333, interpMode=RCIM_Cubic, tangentMode=RCTM_User, time=1, value=0}
         ]
        }}}
      Scale Curve=1.25
     ))
    ScaleA=true
    ScaleRGB=true
   )
   &SubUVAnimation(
    "Sprite Renderer "=DataInterface() {propertyValues {SpriteRenderer=&self:NiagaraEmitter_3.NiagaraSpriteRendererProperties_4}}
    Start Frame Offset:NiagaraInt32=&RandomRangeInt(Minimum=0, Maximum=1)
    UseEndFrame=false
    UseStartFrame=false
   )
  ]
  rendererClasses [&NiagaraSpriteRendererProperties]
  renderers [{rendererClass=&NiagaraSpriteRendererProperties}]
 }
 NE_DROPS001 {
  emitterUpdateScript [
   &EmitterState(Inactive Response=::NewEnumerator1, Life Cycle Mode=::NewEnumerator1, Loop Duration=1, UseLoopDelay=false)
   &VC_FilteredSocketCount()
   &SpawnRate(SpawnRate=50, Use Spawn Probability=false)
  ]
  particleSpawnScript [
   &VC_PaintedVariantsOverride(
    Black=<rgba 0.1 0.1 0.1 1>
    BurntSienna=<rgba 0.55 0.11 0 1>
    Cobalt=<rgba 0 0.4167 1 1>
    Crimson=<rgba 1 0 0 1>
    DefaultColor=<rgba 1 0.0333 0 1>
    ForestGreen=<rgba 0 0.8 0 1>
    Grey=<rgba 0.5 0.5 0.5 1>
    Lime=<rgba 0.925 1.5 0 1>
    Orange=<rgba 1.3 0.1733 0 1>
    Pink=<rgba 1.2 0.36 1.06 1>
    Purple=<rgba 0.4167 0 1 1>
    Saffron=<rgba 1.2133 1.3 0 1>
    SkyBlue=<rgba 0 0.65 1.3 1>
   )
   &InitializeParticle(
    Color=$Particles.Color
    Lifetime Max=0.5
    Lifetime Min=0.2
    Lifetime Mode=::NewEnumerator1
    Sprite Rotation Mode=::NewEnumerator2
    Sprite Size Mode=::NewEnumerator1
    Uniform Sprite Size Max=20
    Uniform Sprite Size Min=15
    UsePositionOffset=false
    Write Material Random=true
   )
   &PartitionParticles(Number Of Partitions=$Emitter.PortCount)
   &SkeletalMeshLocation(
    Bone / Socket Index=$Particles.PartitionParticles.Partition
    Bone / Socket Sampling Mode=::NewEnumerator5
    MaskSpawnGroup=false
    Mesh Sampling Type=::NewEnumerator6
    Skeletal Mesh=$System.SkeletalMesh
   )
   &ShapeLocation(
    MaskSpawnGroup=false
    Offset=&AddVectorToPosition(Position=<-95 0 0>, Vector=$User.DTCarOffset)
    Offset Coordinate Space=::Local
    Offset Mode=::NewEnumerator0
    Override Seed=false
    Sphere Radius=30
    UseOwnerScale=true
   )
   &AddVelocity(Velocity=&RandomRangeVector(Minimum=<-300 0 0>, Maximum=<-1600 0 0>))
  ]
  particleUpdateScript [
   &ParticleState()
   &ScaleSpriteSize(Uniform Curve Sprite Scale=DataInterface() {propertyValues {Curve {keys [{tangentMode=RCTM_User, time=0, value=0.5}, {tangentMode=RCTM_User, time=0.2, value=1}]}}})
   &Collision(
    Age Colliding Particles=false
    Analytical Collision Plane Position 1:Vector3f=<0 0 30>
    Analytical Collision Plane Space:ENiagaraCoordinateSpace=::World
    CPU Collision Type=::NewEnumerator3
    Enable Rest State=false
    KillOnCollision=false
    Module.EnableMaxCollisionCount=false
    Module.Randomize Collision Normal Vector=false
    Perform 2 Analytical Collision Queries=false
    Restitution=0
   )
   /Niagara/Modules/Events/GenerateCollisionEvent(Module.EventDelay=true, Module.Use Velocity Threshold=true)
   &InheritSourceMovement(
    Applied Position Scale=&Multiply_VectorByFloat(
     Vector=<1 1 1>
     Float=&FloatFromCurve(FloatCurve=DataInterface() {propertyValues {Curve {keys [<at 0 1 RCIM_Cubic>, {arriveTangent=-3.1724, interpMode=RCIM_Cubic, leaveTangent=-3.1724, tangentMode=RCTM_User, time=1, value=0}]}}})
    )
    ApplyForce=false
    ApplyPosition=true
    ApplyPositionFalloff=false
   )
   &GravityForce()
   &Drag(Drag=&RandomRangeFloat(Minimum=0.5, Maximum=1), Use Linear Drag=true, Use Rotational Drag=false)
   &SolveForcesAndVelocity(Clamp Velocity=false, Limit Acceleration=false)
   &KillParticles(Kill Particles=$Particles.HasCollided)
   &ScaleColor(
    Scale Alpha=2
    Scale RGB=&VectorFromFloat(Value=&FloatFromCurve(FloatCurve=DataInterface() {propertyValues {Curve {keys [<at 0 0.6 RCIM_Cubic>, {arriveTangent=-0.2897, interpMode=RCIM_Cubic, leaveTangent=-0.2897, tangentMode=RCTM_User, time=1, value=0.4}]}}}))
    ScaleA=true
    ScaleRGB=true
   )
  ]
  rendererClasses [&NiagaraSpriteRendererProperties]
  renderers [{rendererClass=&NiagaraSpriteRendererProperties}]
 }
 NE_FIRE001 {
  emitterUpdateScript [
   &EmitterState(Inactive Response=::NewEnumerator1, Life Cycle Mode=::NewEnumerator1, Loop Duration=0.1, UseLoopDelay=false)
   &SpawnBurst_Instantaneous(Spawn Count=10, Use Spawn Probability=false, UseLoopCountLimit=false)
   &VC_FilteredSocketCount()
   &SpawnRate(SpawnGroup=1, SpawnRate=41, Use Spawn Probability=false)
  ]
  particleSpawnScript [
   &VC_PaintedVariantsOverride(
    Black=<rgba 0.005 0.005 0.005 4>
    BurntSienna=<rgba 0.3 0.06 0 1>
    Cobalt=<rgba 0.064 0.1253 0.8 1>
    Crimson=<rgba 0.5 0 0 1>
    DefaultColor=<rgba 2 0.1167 0 0.5>
    ForestGreen=<rgba 0.045 0.5 0.045 1>
    Grey=<rgba 0.11 0.11 0.11 1>
    Lime=<rgba 0.5242 0.85 0 0.25>
    Orange=<rgba 0.8 0.1067 0 1>
    Pink=<rgba 0.65 0.195 0.5742 1>
    Purple=<rgba 0.0917 0 0.22 2>
    R1 Black:LinearColor=<rgba 0.3 0.3 0.3 1>
    R1 BurntSienna:LinearColor=<rgba 2.5 0.5 0 1>
    R1 Cobalt:LinearColor=<rgba 0.24 0.47 3 1>
    R1 Crimson:LinearColor=<rgba 3 0 0 1>
    R1 DefaultColor=<rgba 3 0.1 0 1>
    R1 ForestGreen:LinearColor=<rgba 0 2.1 0 1>
    R1 Grey:LinearColor=<rgba 2 2 2 1>
    R1 Lime:LinearColor=<rgba 1.85 3 0 1>
    R1 Orange:LinearColor=<rgba 3 0.4 0 1>
    R1 Pink:LinearColor=<rgba 3 0.9 2.65 1>
    R1 Purple:LinearColor=<rgba 1.0417 0 2.5 1>
    R1 Saffron:LinearColor=<rgba 2.8 3 0 1>
    R1 SkyBlue:LinearColor=<rgba 0 1.5 3 1>
    R1 TitaniumWhite:LinearColor=<rgba 3 3 3 1>
    Saffron=<rgba 0.4683 0.5 0.025 0.5>
    SkyBlue=<rgba 0 0.475 0.95 0.8>
    TitaniumWhite=<rgba 0.6 0.6 0.6 0.25>
    Use R1 Painted Overrides=true
   )
   &InitializeParticle(
    Color=$Particles.Color
    Lifetime=&MergeHelper_FloatSwitch(
     0=&RandomRangeFloat(Minimum=0.35, Maximum=0.25)
     1=&RandomRangeFloat(Minimum=0.5, Maximum=0.7)
     2=?
     3=?
     Index=&SpawnGroup()
     Num=::NewEnumerator0
    )
    Mesh Scale Mode=::NewEnumerator1
    Mesh Uniform Scale Max=0.4
    Mesh Uniform Scale Min=0.25
    Position Mode=::NewEnumerator0
    Sprite Rotation Mode=::NewEnumerator2
    Sprite Size Mode=::NewEnumerator1
    Uniform Sprite Size Max=60
    Uniform Sprite Size Min=55
    Write Material Random=true
   )
   &PartitionParticles(Number Of Partitions=$Emitter.PortCount)
   &SkeletalMeshLocation(
    Bone / Socket Index=$Particles.PartitionParticles.Partition
    Bone / Socket Sampling Mode=::NewEnumerator5
    MaskSpawnGroup=false
    Mesh Sampling Type=::NewEnumerator6
    Skeletal Mesh=$System.SkeletalMesh
   )
   &ShapeLocation(
    MaskSpawnGroup=false
    Offset=&AddVectorToPosition(
     Position=&MergeHelper_VectorSwitch(0=<-85 0 0>, 1=<-95 0 0>, 2=?, 3=?, Index=&SpawnGroup(), Num=::NewEnumerator0)
     Vector=$User.DTCarOffset
    )
    Offset Coordinate Space=::Local
    Offset Mode=::NewEnumerator0
    Override Seed=false
    Sphere Radius=&MergeHelper_FloatSwitch(0=0, 1=45, 2=?, 3=?, Index=&SpawnGroup(), Num=::NewEnumerator0)
    UseOwnerScale=true
   )
   &AddVelocity(Velocity=&MergeHelper_VectorSwitch(
     0=&RandomRangeVector(Minimum=<-450 0 0>, Maximum=<-550 0 0>)
     1=&RandomRangeVector(Minimum=<-1100 0 0>, Maximum=<-1300 0 0>)
     2=?
     3=?
     Index=&SpawnGroup()
     Num=::NewEnumerator0
    ))
   &InitialMeshOrientation(Use Rotation Vector=false)
   /SetVariables(Particles.VisibilityTag=$Emitter.SpawnGroup) {isSetParameters=true}
  ]
  particleUpdateScript [
   &ParticleState()
   &ScaleSpriteSize(Uniform Curve Sprite Scale=DataInterface() {propertyValues {Curve {defaultValue=0, keys [<at 0 1.15 RCIM_Cubic>,<at 1 1 RCIM_Cubic>]}}})
   &ScaleMeshSize(Scale Factor=&VectorFromFloat(Value=&FloatFromCurve(FloatCurve=DataInterface() {propertyValues {Curve {keys [<at 0 0.2>,<at 0.15 1>,<at 0.7 1>,<at 1 0>]}}})))
   &Drag(
    Drag=&MergeHelper_FloatSwitch(
     0=&FloatFromCurve(FloatCurve=DataInterface() {propertyValues {Curve {
         defaultValue=0
         keys [
          {arriveTangent=0.1866, interpMode=RCIM_Cubic, leaveTangent=0.1866, tangentMode=RCTM_User, time=0, value=0}
          {interpMode=RCIM_Cubic, tangentMode=RCTM_User, time=0.72, value=0.72}
         ]
        }}})
     1=0
     2=?
     3=?
     Index=&SpawnGroup()
     Num=::NewEnumerator0
    )
    Use Linear Drag=true
    Use Rotational Drag=false
   )
   &InheritSourceMovement(
    Applied Position Scale=&Multiply_VectorByFloat(
     Vector=<1 1 1>
     Float=&FloatFromCurve(FloatCurve=DataInterface() {propertyValues {Curve {keys [<at 0 1 RCIM_Cubic>, {arriveTangent=-3.1552, interpMode=RCIM_Cubic, leaveTangent=-3.1552, tangentMode=RCTM_User, time=1, value=0}]}}})
    )
    ApplyForce=false
    ApplyPosition=true
    ApplyPositionFalloff=false
   )
   &SolveForcesAndVelocity(Clamp Velocity=false, Limit Acceleration=false)
   &ScaleColor(
    Scale RGB=&VectorFromFloat(Value=&FloatFromCurve(
      FloatCurve=DataInterface() {propertyValues {Curve {
         defaultValue=1
         keys [{interpMode=RCIM_Cubic, tangentMode=RCTM_User, time=0, value=1.5}, {interpMode=RCIM_Cubic, tangentMode=RCTM_User, time=1, value=0.5}]
        }}}
      Scale Curve=2
     ))
    ScaleA=true
    ScaleRGB=true
   )
   &SubUVAnimation(
    "Sprite Renderer "=DataInterface() {propertyValues {SpriteRenderer=&self:NiagaraEmitter_1.NiagaraSpriteRendererProperties_1}}
    UseEndFrame=false
    UseStartFrame=false
   )
  ]
  rendererClasses [&NiagaraMeshRendererProperties, &NiagaraSpriteRendererProperties]
  renderers [
   {rendererClass=&NiagaraMeshRendererProperties, rendererIndex=0}
   {rendererClass=&NiagaraSpriteRendererProperties, rendererIndex=1}
  ]
 }
 NE_SWIRL {
  emitterUpdateScript [
   &EmitterState(Inactive Response=::NewEnumerator1, Life Cycle Mode=::NewEnumerator1, Loop Duration=0.1, UseLoopDelay=false)
   &VC_FilteredSocketCount()
   &SpawnRate(SpawnRate=15, Use Spawn Probability=false)
   &SpawnBurst_Instantaneous(Spawn Count=2, Spawn Group=1, Use Spawn Probability=false, UseLoopCountLimit=false)
  ]
  particleSpawnScript [
   &VC_PaintedVariantsOverride(
    Black=<rgba 0.05 0.05 0.05 1>
    BurntSienna=<rgba 0.25 0.05 0 1>
    Cobalt=<rgba 0.05 0.125 0.5 1>
    Crimson=<rgba 0.3 0 0 1>
    DefaultColor=<rgba 0.5 0.0167 0 1>
    ForestGreen=<rgba 0 0.3 0 1>
    Grey=<rgba 0.1 0.1 0.1 1>
    Lime=<rgba 0.185 0.3 0 1>
    Orange=<rgba 0.4 0.0533 0 1>
    Pink=<rgba 0.35 0.105 0.3092 1>
    Purple=<rgba 0.2083 0 0.5 1>
    R1 Black:LinearColor=<rgba 0.05 0.05 0.05 1>
    R1 BurntSienna:LinearColor=<rgba 0.55 0.11 0 1>
    R1 Cobalt:LinearColor=<rgba 0.16 0.3133 2 1>
    R1 Crimson:LinearColor=<rgba 1 0 0 1>
    R1 DefaultColor=<rgba 2 0.0667 0 1>
    R1 ForestGreen:LinearColor=<rgba 0 0.8 0 1>
    R1 Grey:LinearColor=<rgba 0.5 0.5 0.5 1>
    R1 Lime:LinearColor=<rgba 0.925 1.5 0 1>
    R1 Orange:LinearColor=<rgba 1.3 0.1733 0 1>
    R1 Pink:LinearColor=<rgba 1.2 0.36 1.06 1>
    R1 Purple:LinearColor=<rgba 0.4167 0 1 1>
    R1 Saffron:LinearColor=<rgba 1.2133 1.3 0 1>
    R1 SkyBlue:LinearColor=<rgba 0 0.65 1.3 1>
    Saffron=<rgba 0.28 0.3 0 1>
    SkyBlue=<rgba 0 0.25 0.5 1>
    TitaniumWhite=<rgba 0.3 0.3 0.3 1>
    Use R1 Painted Overrides=true
   )
   &InitializeParticle(
    Color=$Particles.Color
    Lifetime=&MergeHelper_FloatSwitch(0=1, 1=0.3, 2=?, 3=?, Index=&SpawnGroup(), Num=::NewEnumerator0)
    Mesh Scale=&Multiply_VectorByFloat(Vector=<1 0.3 0.3>, Float=1)
    Mesh Scale Mode=::NewEnumerator4
    Position Mode=::NewEnumerator0
    Sprite Rotation Mode=::NewEnumerator2
    Sprite Size=<2 15>
    Sprite Size Mode=::NewEnumerator4
    Write Material Random=true
   )
   &PartitionParticles(Number Of Partitions=$Emitter.PortCount)
   &SkeletalMeshLocation(
    Bone / Socket Index=$Particles.PartitionParticles.Partition
    Bone / Socket Sampling Mode=::NewEnumerator5
    MaskSpawnGroup=false
    Mesh Sampling Type=::NewEnumerator6
    Sampled Position Offset=&AddVectorToPosition(
     Position=&MergeHelper_VectorSwitch(
      0=&RandomRangeVector(Minimum=<-90 -30 -50>, Maximum=<-90 30 50>)
      1=&RandomRangeVector(Minimum=<-60 -1 -1>, Maximum=<-60 1 1>)
      2=?
      3=?
      Index=&SpawnGroup()
      Num=::NewEnumerator0
     )
     Vector=$User.DTCarOffset
    )
    Skeletal Mesh=$System.SkeletalMesh
   )
   &AddVelocity(
    Velocity=&MaskVectorBySpawnGroup(
     Masked Vector=<0 0 0>
     Passthrough Vector=&RandomRangeVector(Minimum=<150 -250 -100>, Maximum=<300 250 100>)
     Spawn Group Mask=1
    )
    Velocity Speed Scale=-1
   )
   /SetVariables(Particles.VisibilityTag=$Emitter.SpawnGroup) {isSetParameters=true}
   &InitialMeshOrientation(
    Mesh Orientation Mode=::NewEnumerator2
    Rotation=&Lerp_Vector(A=<-1 0 0>, B=<1 0 0>, Alpha=&Sine(Normalized Angle=$Emitter.Age, Period=0.1, Scale=1, Bias=0))
    Use Rotation Vector=true
   )
  ]
  particleUpdateScript [
   &ParticleState()
   &ScaleSpriteSize(
    Uniform Curve Scale=1.2
    Uniform Curve Sprite Scale=DataInterface() {propertyValues {Curve {keys [
        {arriveTangent=-0.0405, interpMode=RCIM_Cubic, leaveTangent=-0.0405, tangentMode=RCTM_User, time=0.3, value=1}
        {arriveTangent=0.0388, interpMode=RCIM_Cubic, leaveTangent=0.0388, tangentMode=RCTM_User, time=1, value=0}
       ]}}}
   )
   &ScaleMeshSize(Scale Factor=&Multiply_Vector(
     A=&VectorFromFloat(Value=1)
     B=&MakeVector(
      X=1
      Y=&MergeHelper_FloatSwitch(0=1, 1=-1, 2=?, 3=?, Index=$Particles.PartitionParticles.Partition, Num=::NewEnumerator0)
      Z=1
     )
    ))
   &VortexForce(
    Origin Pull Amount:NiagaraFloat=&MaskFloatBySpawnGroup(Masked Float=&RandomRangeFloat(Minimum=10000, Maximum=10000), Passthrough Float=0)
    UseOriginOffset=true
    Vortex Axis=&RandomRangeVector(Minimum=<1 0 0>, Maximum=<-1 0 0>, Evaluation Type=::NewEnumerator1)
    Vortex Axis Coordinate Space:ENiagaraCoordinateSpace=::Local
    Vortex Force Amount=&MaskFloatBySpawnGroup(Masked Float=&RandomRangeFloat(Minimum=7000, Maximum=10000), Passthrough Float=0)
    Vortex Origin:NiagaraPosition=$Particles.Initial.Position
    Vortex Origin Offset:Vector3f=&RandomRangeVector(Minimum=<0 0 0>, Maximum=<-20 0 0>)
    Vortex Origin Offset Coordinate Space:ENiagaraCoordinateSpace=::Local
   )
   &SolveForcesAndVelocity(Clamp Velocity=false, Limit Acceleration=false)
   &ScaleColor(Scale Alpha=5, Scale RGB=&VectorFromFloat(Value=5), ScaleA=true, ScaleRGB=true)
   /SetVariables(Particles.sparkscolor:LinearColor=&MultiplyLinearColor(
     Linear Color=$Particles.R1 Color
     Scale Factor=&MakeVector4FromVectorAndFloat(
      "Vector (RGB)"=&VectorFromFloat(Value=5)
      "Float (Alpha)"=&FloatFromCurve(
       FloatCurve=DataInterface() {propertyValues {Curve {
          defaultValue=1
          keys [
           {arriveTangent=1.0371, interpMode=RCIM_Cubic, leaveTangent=1.0371, leaveTangentWeight=0.1667, tangentMode=RCTM_User, time=0, value=0}
           {arriveTangentWeight=0.1667, interpMode=RCIM_Cubic, leaveTangentWeight=0.1667, tangentMode=RCTM_User, time=0.5, value=1}
           {arriveTangent=-3.3261, arriveTangentWeight=0.1667, interpMode=RCIM_Cubic, leaveTangent=-3.3261, tangentMode=RCTM_User, time=1, value=0}
          ]
         }}}
       Scale Curve=2
      )
     )
    )) {isSetParameters=true}
  ]
  rendererClasses [&NiagaraMeshRendererProperties, &NiagaraSpriteRendererProperties]
  renderers [
   {rendererClass=&NiagaraMeshRendererProperties, rendererIndex=0}
   {rendererClass=&NiagaraSpriteRendererProperties, rendererIndex=1}
  ]
 }
 NE_TRAIL {
  emitterUpdateScript [
   &EmitterState(Inactive Response=::NewEnumerator1, Life Cycle Mode=::NewEnumerator1, Loop Duration=1, UseLoopDelay=false)
   &VC_FilteredSocketCount()
   &VC_RateQualityHandler(Epic Spawn Rate=50, High Spawn Rate=40, Low Spawn Rate=30, Use Custom Rate?=true)
   &VC_SpawnRibbon_v1(Number of Ribbons=$Emitter.PortCount, Spawn Time=-1, SpawnRate=$Emitter.Spawn Rate)
  ]
  particleSpawnScript [
   &VC_PaintedVariantsOverride(
    Black=<rgba 0.1 0.1 0.1 1>
    BurntSienna=<rgba 0.6 0.12 0 1>
    Cobalt=<rgba 0.16 0.3133 2 1>
    Crimson=<rgba 1.1 0 0 1>
    DefaultColor=<rgba 2 0.0667 0 1>
    ForestGreen=<rgba 0 0.8 0 1>
    Grey=<rgba 0.5 0.5 0.5 1>
    Lime=<rgba 0.925 1.5 0 1>
    Orange=<rgba 1.3 0.1733 0 1>
    Pink=<rgba 1.2 0.36 1.06 1>
    Purple=<rgba 0.4167 0 1 1>
    R1 Black:LinearColor=<rgba 0.05 0.05 0.05 1>
    R1 BurntSienna:LinearColor=<rgba 0.05 0.05 0.05 1>
    R1 Cobalt:LinearColor=<rgba 0 0.007 0.5 1>
    R1 Crimson:LinearColor=<rgba 0.05 0.05 0.05 1>
    R1 DefaultColor=<rgba 0.5 0.0575 0.05 1>
    R1 ForestGreen:LinearColor=<rgba 0.05 0.05 0.05 1>
    R1 Gold:LinearColor=<rgba 0.05 0.05 0.05 1>
    R1 Grey:LinearColor=<rgba 0.05 0.05 0.05 1>
    R1 Lime:LinearColor=<rgba 0.05 0.05 0.05 1>
    R1 Orange:LinearColor=<rgba 0.05 0.05 0.05 1>
    R1 Pink:LinearColor=<rgba 0.05 0.05 0.05 1>
    R1 Purple:LinearColor=<rgba 0.05 0.05 0.05 1>
    R1 Saffron:LinearColor=<rgba 0.05 0.05 0.05 1>
    R1 SkyBlue:LinearColor=<rgba 0.05 0.05 0.05 1>
    R1 TitaniumWhite:LinearColor=<rgba 1 1 1 1>
    Saffron=<rgba 1.2133 1.3 0 1>
    SkyBlue=<rgba 0 0.65 1.3 1>
    Use R1 Painted Overrides=true
   )
   &VC_InitializeRibbon_v1(
    Color=$Particles.Color
    Custom Offset=true
    Enable Skeletal Position=true
    Lifetime=0.7
    Local Vector Offset=&AddVectorToPosition(Position=<0 0 0>, Vector=$User.DTCarOffset)
    Offset Amount=1
    Ribbon Width=30
   )
   &AddVelocity(Velocity=<-2500 0 0>) {module=AddVelocity001}
  ]
  particleUpdateScript [
   &ParticleState()
   &InheritSourceMovement(ApplyForce=false, ApplyPosition=true, ApplyPositionFalloff=false)
   &SolveForcesAndVelocity(Clamp Velocity=false, Limit Acceleration=false)
   &VC_UpdateRibbon_v1(Enable Skeletal Position=true)
   &ScaleRibbonWidth(Ribbon Width Scale=&FloatFromCurve(FloatCurve=DataInterface() {propertyValues {Curve {keys [<at 0 1>, {leaveTangent=-0.2059, time=1, value=0.5}]}}}))
   /SetVariables(
    Particles.ScaleColor_bASE:LinearColor=&MultiplyLinearColor(
     Linear Color=$Particles.R1 Color
     Scale Factor=&MakeVector4FromVectorAndFloat(
      "Vector (RGB)"=&VectorFromFloat(Value=&FloatFromCurve(FloatCurve=DataInterface() {propertyValues {Curve {
           defaultValue=1
           keys [{interpMode=RCIM_Cubic, tangentMode=RCTM_User, time=0.5, value=1}, {interpMode=RCIM_Cubic, tangentMode=RCTM_User, time=1, value=0}]
          }}}))
      "Float (Alpha)"=1
     )
    )
    Particles.width2:NiagaraFloat=&Multiply_Float(
     A=35
     B=&FloatFromCurve(FloatCurve=DataInterface() {propertyValues {Curve {keys [<at 0 1>, {leaveTangent=-0.2059, time=1, value=0.7}]}}})
    )
   ) {isSetParameters=true}
   &ScaleColor(
    Scale RGB=&VectorFromFloat(Value=&FloatFromCurve(
      FloatCurve=DataInterface() {propertyValues {Curve {keys [{arriveTangent=-14.9108, interpMode=RCIM_Cubic, leaveTangent=-14.9108, tangentMode=RCTM_User, time=0.15, value=5}, <at 0.5 0.4 RCIM_Cubic>]}}}
      Scale Curve=2
     ))
    ScaleA=true
    ScaleRGB=true
   )
  ]
  rendererClasses [&NiagaraRibbonRendererProperties]
  renderers [
   {rendererClass=&NiagaraRibbonRendererProperties}
   {rendererClass=&NiagaraRibbonRendererProperties, rendererIndex=1}
  ]
 }
}
summary {
 systemName=NS_Boost_Magma
 emitters [
  {emitterName=Collided_blobs, rendererClasses [&NiagaraMeshRendererProperties, &NiagaraSpriteRendererProperties]}
  {emitterName=NE_TRAIL, rendererClasses [&NiagaraRibbonRendererProperties]}
  {emitterName=NE_SWIRL, rendererClasses [&NiagaraMeshRendererProperties, &NiagaraSpriteRendererProperties]}
  {emitterName=NE_FIRE001, rendererClasses [&NiagaraMeshRendererProperties, &NiagaraSpriteRendererProperties]}
  {emitterName=NE_DROPS001, rendererClasses [&NiagaraSpriteRendererProperties]}
  {emitterName=NE_CORE_BURST, rendererClasses [&NiagaraSpriteRendererProperties]}
 ]
 userVariables [
  User.bIsFrontendPreview {type=NiagaraBool}
  User.bIsPainted {type=NiagaraBool}
  User.ColorVariantID {type=NiagaraInt32}
  User.DTCarOffset {type=Vector3f, description="", defaultValue {struct=Vector3f, value=<0 0 0>}}
  User.PaintedColor {type=LinearColor, description="", defaultValue {struct=LinearColor, value=<rgba 1 1 1 1>}}
  User.PaintedScalar {type=NiagaraFloat, defaultValue=1, description=""}
  User.PortCount {type=NiagaraInt32, defaultValue=2}
 ]
}
