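# Unlit material for Niagara mesh particles, apparently the collision blobs of the Magma boost effect.
# A rotated, panning RGB mask blends red-orange tints, multiplied by particle color and Intensity, then fed to MF_FX_EmissiveAdjustments.
# The Painted switch desaturates and sharpens the pattern so particle color sets the hue.
# A separate branch builds a wobble-style offset vector from animated sine products, scaled by strength and a vertical UV mask.
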Material /VehicleCosmetics/Boosts/Boost_Magma/Material/M_Collision_Blobs
graph {
 properties {
  bAutomaticallySetUsageInEditor=false
  bUsedWithNiagaraMeshParticles=true
  pixelDepthOffsetMode=PDOM_Legacy
  shadingModel=MSM_Unlit
  usedShadingModels=Unlit
 }
 nodes [
  MaterialExpressionMeshEmitterVertexColor_0 = ParticleColor()
  n1 = TextureCoordinate()
  n2 = TextureSample(UVs=Panner(Coordinate=Rotator(Coordinate=@n1)) {speedX=1}) {texture=/DealershipSharedLibrary/FX/Textures/TexturePackage/T_VCVFX_ColorPaning_RGB}
  n3 = Constant3Vector() {constant=<rgba 1 0.1 0 0>}
  n4 = Constant3Vector() {constant=<rgba 0.4 0 0 0>}
  n5 = Add(A=Multiply(A=Constant3Vector() {constant=<rgba 1 0.1667 0 0>}, B=@n2.R), B=Add(A=Multiply(A=@n3, B=@n2.G), B=Multiply(A=@n2.B, B=@n4)))
  n6 = Desaturation(None=@n5, Fraction=Constant() {r=1})
  n7 = Constant() {r=5}
  n8 = Multiply(A=@MaterialExpressionMeshEmitterVertexColor_0.RGB, B=Power(Base=Multiply(A=@n6, B=@n7), Exp=Constant() {r=2}))
  n9 = Multiply(A=@MaterialExpressionMeshEmitterVertexColor_0.RGB, B=@n5)
  MaterialFunctionCall(Emissive=Multiply(A=StaticSwitchParameter(True=@n8, False=@n9) {parameterName=Painted}, B=ScalarParameter() {defaultValue=4, parameterName=Intensity})) {materialFunction=&MF_FX_EmissiveAdjustments}
  n10 = Constant()
  n11 = Divide(A=ComponentMask(None=TextureCoordinate() {coordinateIndex=1}) {r=true}, B=ScalarParameter() {defaultValue=1, parameterName=CountWaves2})
  n12 = Multiply(A=ScalarParameter() {defaultValue=0.5, parameterName=HorizontalSpeed}, B=OneMinus(None=Time()))
  n13 = Multiply(A=ScalarParameter() {defaultValue=-0.2, parameterName=HorizontalSpeed2}, B=OneMinus(None=Time()))
  n14 = Multiply(A=Sine(None=Add(A=@n11, B=@n12)) {period=0.5}, B=Sine(None=Add(B=@n13)) {period=0.5})
  n15 = Constant() {r=1.2}
  n16 = TextureCoordinate() {coordinateIndex=1}
  n17 = Time()
  n18 = Add(
   A=Divide(A=ComponentMask(None=@n16) {g=true}, B=ScalarParameter() {defaultValue=1, parameterName=CountWaves2})
   B=Multiply(A=ScalarParameter() {defaultValue=2, parameterName=VerticalSpeed}, B=OneMinus(None=@n17))
  )
  n19 = Time()
  n20 = Add(B=Multiply(A=ScalarParameter() {defaultValue=-0.5, parameterName=VerticalSpeed2}, B=OneMinus(None=@n19)))
  n21 = Multiply(A=OneMinus(None=Sine(None=@n18) {period=0.5}), B=OneMinus(None=Sine(None=@n20) {period=0.5}))
  n22 = Constant() {r=1.2}
  n23 = Multiply(A=AppendVector(A=Power(Base=@n14, Exp=@n15), B=Power(Base=@n21, Exp=@n22)), B=ScalarParameter() {defaultValue=2, parameterName=Strenght})
  n24 = Power(Base=ScalarParameter() {defaultValue=1, parameterName=PowerMiddle2})
  n25 = TextureCoordinate()
  n26 = Multiply(
   A=Multiply(A=OneMinus(None=@n24), B=ComponentMask(None=@n25) {g=true})
   B=ScalarParameter() {defaultValue=6, parameterName=MultPower2Vertical}
  )
  Multiply(A=AppendVector(A=AppendVector(A=@n10, B=@n10), B=Multiply(A=@n23, B=@n26)), B=Constant() {r=0.15})
 ]
}
