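# M_Embers is a translucent, unlit Niagara sprite material for ember particles in the Magma boost vehicle cosmetic.
# Particle color RGB feeds the emissive input of the shared MF_FX_EmissiveAdjustments function, so the particle system drives the glow color.
# Particle alpha multiplies the RGB of a shared ash mask texture, forming a particle-faded sprite shape.
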Material /VehicleCosmetics/Boosts/Boost_Magma/Material/M_Embers
graph {
 properties {bUsedWithNiagaraSprites=true, blendMode=BLEND_Translucent, pixelDepthOffsetMode=PDOM_Legacy, shadingModel=MSM_Unlit, usedShadingModels=Unlit}
 nodes [
  n1 = ParticleColor()
  MaterialFunctionCall(Emissive=@n1.RGB) {materialFunction=&MF_FX_EmissiveAdjustments}
  Multiply(A=@n1.A, B=TextureSample() {texture=/DealershipSharedLibrary/FX/Textures/Masks/T_VCVFX_Ash_A}.RGB)
 ]
}
